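# Sound cue for the end of the suppressed sniper's reload, played as weapon foley.
# A Random node picks among three equally weighted reload-end waves and feeds a Modulator at the root.
# The cue lowers volume, slightly raises pitch, uses the shared weapon reload attenuation and overrides concurrency.
# The debug flag is left enabled.
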SoundCue /WeaponsUnvaulted/Assets/Audio/SuppressedSniper/SC_SuppressedSniper_ReloadEnd
root=Modulator_0
nodes [
 {id=Modulator_0, children [Random_0]}
 {id=Random_0, children [WavePlayer_0,WavePlayer_1,WavePlayer_3], properties {weights [1,1,1]}}
 {id=WavePlayer_0, properties {soundWaveAssetPtr=/WeaponsUnvaulted/Assets/Audio/SuppressedSniper/SoundWaves/SniperSuppressed_Reload_End_01}}
 {id=WavePlayer_1, properties {soundWaveAssetPtr=/WeaponsUnvaulted/Assets/Audio/SuppressedSniper/SoundWaves/SniperSuppressed_Reload_End_02}}
 {id=WavePlayer_3, properties {soundWaveAssetPtr=/WeaponsUnvaulted/Assets/Audio/SuppressedSniper/SoundWaves/SniperSuppressed_Reload_End_03}}
]
properties {
 attenuationSettings=&Fort_Weapon_Reload_Att
 bDebug=true
 bOverrideConcurrency=true
 maxDistance=2250
 pitchMultiplier=1.1
 soundClassObject=&Fort_WeaponFoley
 volumeMultiplier=0.35
}
